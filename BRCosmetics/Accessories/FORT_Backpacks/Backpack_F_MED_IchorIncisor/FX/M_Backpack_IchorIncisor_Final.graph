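# Unlit, premultiplied-alpha Niagara mesh particle material for the Ichor Incisor back bling, drawing animated layered wing shapes.
# A cross-section texture and UV falloff, with sine-animated exponents, form three soft masks tinted outer, middle and inner, plus outer glow and particle color.
# World position offset sums normal puff, a clamped bend driven by Niagara dynamic parameters, and sine waves along instance Y.
# Dynamic parameters also offset the shape animation timing.
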Material /BRCosmetics/Accessories/FORT_Backpacks/Backpack_F_MED_IchorIncisor/FX/M_Backpack_IchorIncisor_Final
graph {
 properties {
  bUsedWithNiagaraMeshParticles=true
  blendMode=BLEND_AlphaComposite
  pixelDepthOffsetMode=PDOM_Legacy
  shadingModel=MSM_Unlit
  usedShadingModels=Unlit
 }
 nodes [
  n1 = TextureCoordinate()
  n2 = ConstantBiasScale(None=ComponentMask(None=@n1) {g=true}) {bias=-0.5, scale=2}
  n3 = Abs(None=ConstantBiasScale(None=ComponentMask(None=@n1) {r=true}) {bias=-0.5, scale=2})
  n4 = Saturate(None=Divide(A=OneMinus(None=@n3), B=ScalarParameter() {defaultValue=0.1666, parameterName=Tips Fade Size}))
  n5 = OneMinus(None=Abs(None=Divide(A=@n2, B=@n4)))
  n6 = VectorParameter() {parameterName=Layers Exponents, defaultValue=<rgba 1.2 1.4 1.4 0>}
  n7 = Time()
  n8 = DynamicParameter() {defaultValue=<rgba 1 1 1 1>, paramNames [ClampInterpolate, WingOffset_01, WingOffset_02, WPO_Wave_Direction]}
  n9 = TextureCoordinate() {coordinateIndex=1}
  n10 = Subtract(
   A=Multiply(A=VectorParameter() {parameterName=Shapes Animation Speeds_2, defaultValue=<rgba 1.5 2 2 0>}.RGB, B=Add(A=@n7, B=@n8.WingOffset_02))
   B=Multiply(A=ComponentMask(None=@n9) {r=true}, B=VectorParameter() {parameterName=Shapes Time Offset Lengths_2, defaultValue=<rgba 2.3 4.2 5.4 0>}.RGB)
  )
  n11 = MaterialFunctionCall(Float3=ConstantBiasScale(None=Sine(None=@n10))) {materialFunction=&BreakOutFloat3Components}
  n12 = MaterialFunctionCall(
   X=Power(Base=@n5, Exp=LinearInterpolate(B=@n6.R, Alpha=@n11.R) {constA=1, constB=1.6})
   Y=Power(Base=@n5, Exp=LinearInterpolate(B=@n6.G, Alpha=@n11.G) {constA=1})
   Z=Power(Base=@n5, Exp=LinearInterpolate(B=@n6.B, Alpha=@n11.B) {constA=1, constB=1.6})
  ) {materialFunction=&MakeFloat3}
  n13 = ComponentMask(None=@n1) {r=true}
  n14 = TextureSample(UVs=AppendVector(A=@n13, B=Constant() {r=0.125})) {
   samplerType=SAMPLERTYPE_LinearColor
   texture=&/T_Backpack_IchorIncisor_CrossSection
  }
  n15 = Time()
  n16 = DynamicParameter() {defaultValue=<rgba 1 1 1 1>, paramNames [ClampInterpolate, WingOffset_01, WingOffset_02, WPO_Wave_Direction]}
  n17 = TextureCoordinate() {coordinateIndex=1}
  n18 = Subtract(
   A=Multiply(A=VectorParameter() {parameterName=Shapes Animation Speeds, defaultValue=<rgba 1 1.3 2 0>}.RGB, B=Add(A=@n15, B=@n16.WingOffset_01))
   B=Multiply(A=ComponentMask(None=@n17) {r=true}, B=VectorParameter() {parameterName=Shapes Time Offset Lengths, defaultValue=<rgba 2.3 4.2 5.4 0>}.RGB)
  )
  n19 = MaterialFunctionCall(Float3=ConstantBiasScale(None=Sine(None=@n18))) {materialFunction=&BreakOutFloat3Components}
  n20 = LinearInterpolate(A=@n14.R, B=@n14.R, Alpha=@n19.R)
  n21 = LinearInterpolate(A=@n14.B, B=@n14.B, Alpha=@n19.R)
  n22 = Saturate(None=Sign(None=@n2))
  n23 = TextureSample(UVs=AppendVector(A=@n13, B=Constant() {r=0.375})) {
   samplerType=SAMPLERTYPE_LinearColor
   texture=&T_Backpack_IchorIncisor_CrossSection
  }
  n24 = LinearInterpolate(A=@n23.R, B=@n23.R, Alpha=@n19.G)
  n25 = LinearInterpolate(A=@n23.B, B=@n23.B, Alpha=@n19.G)
  n26 = TextureSample(UVs=AppendVector(A=@n13, B=Constant() {r=0.625})) {
   samplerType=SAMPLERTYPE_LinearColor
   texture=&T_Backpack_IchorIncisor_CrossSection
  }
  n27 = LinearInterpolate(A=@n26.R, B=@n26.R, Alpha=@n19.B)
  n28 = LinearInterpolate(A=@n26.B, B=@n26.B, Alpha=@n19.B)
  n29 = Subtract(
   A=Divide(A=OneMinus(None=@n12.Result), B=VectorParameter() {parameterName=Layers Sizes, defaultValue=<rgba 1 1 1 0>}.RGB)
   B=MaterialFunctionCall(
    X=LinearInterpolate(A=@n20, B=@n21, Alpha=@n22)
    Y=LinearInterpolate(A=@n24, B=@n25, Alpha=@n22)
    Z=LinearInterpolate(A=@n27, B=@n28, Alpha=@n22)
   ) {materialFunction=&MakeFloat3}.Result
  )
  n30 = OneMinus(None=@n5)
  n31 = VectorParameter() {parameterName=Layers Softness, defaultValue=<rgba 0.2 1 2 0>}
  n32 = Saturate(None=Divide(A=Multiply(A=@n29) {constB=-1}, B=Multiply(A=@n30, B=@n31.RGB)))
  n33 = ComponentMask(None=@n32) {r=true}
  n34 = VectorParameter() {parameterName=Color Outer Shape, defaultValue=<rgba 0.15 0.0261 0.137 0>}
  n35 = VectorParameter() {parameterName=Color Middle Shape, defaultValue=<rgba 3 0.2 0 1>}
  n36 = ComponentMask(None=@n32) {g=true}
  n37 = Multiply(
   A=@n33
   B=LinearInterpolate(
    A=LinearInterpolate(A=@n34.RGB, B=@n35.RGB, Alpha=@n36)
    B=VectorParameter() {parameterName=Color Inner Shape, defaultValue=<rgba 7 0.924 0.49 1>}.RGB
    Alpha=ComponentMask(None=@n32) {b=true}
   )
  ) {desc=Because Premultiplied blending}
  n38 = VectorParameter() {parameterName=Outer Glow Color and Softness, defaultValue=<rgba 0.24 0 0.3 2>}
  n39 = Divide(A=@n5, B=@n38.A)
  n40 = Multiply(A=Multiply(A=Saturate(None=@n39), B=OneMinus(None=@n33)), B=@n38.RGB)
  MaterialFunctionCall(Emissive=Multiply(A=Add(A=@n37, B=@n40) {desc=Adding outer glow}, B=ParticleColor().RGB)) {materialFunction=&MF_FX_EmissiveAdjustments}
  n41 = ConstantBiasScale(None=ComponentMask(None=TextureCoordinate()) {g=true}) {bias=-0.5, scale=2}
  n42 = ScalarParameter() {defaultValue=0.5}
  n43 = Subtract(A=OneMinus(None=Abs(None=@n41)), B=@n42)
  n44 = OneMinus(None=@n42)
  n45 = SquareRoot(None=Saturate(None=Divide(A=@n43, B=@n44)))
  n46 = ConstantBiasScale(None=ComponentMask(None=TextureCoordinate()) {r=true}) {bias=-0.5, scale=2}
  n47 = MaterialFunctionCall(In=OneMinus(None=Abs(None=@n46)), Power=ScalarParameter() {defaultValue=2, parameterName=SCurve_Power1}) {materialFunction=&SCurve}
  n48 = Multiply(
   A=Multiply(A=Multiply(A=@n45, B=@n47.Result), B=VertexNormalWS())
   B=Multiply(A=ScalarParameter() {defaultValue=1.2, parameterName=PuffAmount})
  )
  n49 = DynamicParameter() {desc=Parameters set by Niagara, parameterIndex=1, defaultValue=<rgba 1 1 50 1>, paramNames [WPO UV Scale, "", "", ""]}
  n50 = ParticlePositionWS()
  n51 = Subtract(
   A=MaterialFunctionCall(X=@n49.WPO UV Scale, Y=@n49.WPO UV Scale, Z=ComponentMask(None=@n50) {b=true}) {materialFunction=&MakeFloat3}.Result
   B=@n50
  )
  n52 = ScalarParameter() {defaultValue=10, group="☄WPO", parameterName=Interpolate Clamp}
  n53 = Clamp(None=@n51, Min=Multiply(B=@n52) {constA=-1}, Max=@n52) {maxDefault=1000, minDefault=-1000}
  n54 = DynamicParameter() {defaultValue=<rgba 1 1 1 1>, paramNames [ClampInterpolate, WingOffset_01, WingOffset_02, WPO_Wave_Direction]}
  n55 = Clamp(None=@n51, Min=Multiply(B=@n54.ClampInterpolate) {constA=-1}, Max=@n54.ClampInterpolate) {maxDefault=1000, minDefault=-1000}
  n56 = ScalarParameter() {defaultValue=1, group="☄WPO", parameterName=UseDynamicParameterClampInterpolate}
  n57 = TextureCoordinate()
  n58 = Power(Base=ComponentMask(None=OneMinus(None=@n57)) {r=true}, Exp=ScalarParameter() {defaultValue=2, group="☄WPO", parameterName=WPOBendPower})
  n59 = StaticSwitchParameter(
   True=LinearInterpolate(B=LinearInterpolate(A=@n53, B=@n55, Alpha=@n56) {desc=In case Diving causes issues}, Alpha=Multiply(A=@n58, B=@n49.A))
   False=Constant()
  ) {defaultValue=true, group=WPO, parameterName=Use_Interpolate WPO}
  n60 = ComponentMask(None=TextureCoordinate()) {r=true}
  n61 = OneMinus(None=Abs(None=ConstantBiasScale(None=@n60) {bias=-0.5, scale=2}))
  n62 = ScalarParameter() {defaultValue=0.3, group="☄WPO", parameterName=WaveIntensity}
  n63 = TextureCoordinate()
  n64 = Add(
   A=Multiply(A=ComponentMask(None=@n63) {r=true}, B=ScalarParameter() {defaultValue=1.5, group="☄WPO", parameterName=WaveFrequency})
   B=Multiply(A=Time(), B=ScalarParameter() {defaultValue=1, group="☄WPO", parameterName=WaveSpeed})
  )
  n65 = ScalarParameter() {defaultValue=0.2, group="☄WPO", parameterName=Wave2Intensity}
  n66 = TextureCoordinate()
  n67 = Add(
   A=Multiply(A=ComponentMask(None=@n66) {r=true}, B=ScalarParameter() {defaultValue=4, group="☄WPO", parameterName=Wave2Frequency})
   B=Multiply(A=Time(), B=ScalarParameter() {defaultValue=2, group="☄WPO", parameterName=Wave2Speed})
  )
  n68 = LinearInterpolate(
   A=Multiply(A=Multiply(A=@n62) {constB=0.01, desc=just makes it easier to manipulate}, B=Sine(None=@n64))
   B=Multiply(A=Multiply(A=@n65) {constB=0.01, desc=just makes it easier to manipulate}, B=Sine(None=@n67))
  )
  n69 = Multiply(
   A=Transform(None=Constant3Vector() {constant=<rgba 0 1 0 0>}) {transformSourceType=TRANSFORMSOURCE_Instance}
   B=Multiply(A=Multiply(A=@n61, B=@n68), B=ScalarParameter() {defaultValue=400, parameterName=OverallWaveIntensity})
  )
  NamedRerouteDeclaration(None=Add(A=Add(A=@n48, B=@n59), B=Multiply(A=@n69) {desc=invert directions for the layers})) {name=WPO, nodeColor=<rgba 1 1 1 1>}
  NamedRerouteDeclaration(None=Floor(None=@n33)) {name=OpacityMask, nodeColor=<rgba 1 0 0.9375 1>}
  NamedRerouteUsage()
  NamedRerouteUsage()
  Reroute(None=@n33)
 ]
}
