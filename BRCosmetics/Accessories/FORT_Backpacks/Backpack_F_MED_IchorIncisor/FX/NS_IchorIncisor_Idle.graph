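# Idle effect for the Ichor Incisor backpack, drawing mirrored wing meshes and a flickering purple light attached to the skeletal mesh.
# Glow001 spawns two mesh particles mirrored by spawn group, angled by backward movement, state and skydiving timelines.
# A moving average of position feeds particle attributes, while material parameters vary with timelines, skydiving, vehicle and front-end state.
# Glow supplies the pulsing light; a debug sprite emitter is disabled.
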NiagaraSystem /BRCosmetics/Accessories/FORT_Backpacks/Backpack_F_MED_IchorIncisor/FX/NS_IchorIncisor_Idle
dependencies {
 usedDataInterfaces [&NiagaraDataInterfaceCurve, &NiagaraDataInterfaceSkeletalMesh]
 usedDynamicInputs [
  &VectorLength
  &Add_Float
  &RemapRange
  &OneMinusFloat
  &MakeFloatFromVector
  &SetBoolByFloatComparison
  &Abs_Float
 ]
 usedModules [
  &self:SystemSpawnScript.SystemScriptSource.SystemScriptGraph.NiagaraNodeAssignment_0.SetVariables_#1
  &SystemState
  &self:SystemSpawnScript.SystemScriptSource.SystemScriptGraph.NiagaraNodeAssignment_2.SetVariables_#2
  &self:SystemSpawnScript.SystemScriptSource.SystemScriptGraph.NiagaraNodeAssignment_1.SetVariables_#3
  &Timeline
  &self:SystemSpawnScript.SystemScriptSource.SystemScriptGraph.NiagaraNodeAssignment_3.SetVariables_#4
 ]
}
emitters {
 FN_Debug {
  bEnabled=false
  emitterUpdateScript [
   &EmitterState()
   &SpawnBurst_Instantaneous(Spawn Count=1, Use Spawn Probability=false)
  ]
  particleSpawnScript [/Niagara/Modules/Spawn/Initialization/InitializeParticle(
    Lifetime=5
    Module.Write Color=true
    Module.Write Lifetime=true
    Module.Write Mass=true
    Module.Write Material Random=true
    Module.Write Position=true
    Module.Write Scale=false
    Module.Write SpriteRotation=false
    Module.Write SpriteSize=true
    Position=&AddVectorToPosition(Position=/Niagara/DynamicInputs/Helpers/SimulationPosition(), Vector=<-50 0 0>)
    Sprite Size=<100 100>
   )]
  particleUpdateScript [
   &ParticleState()
   &InterpolateOverTime(Rate Of Change=8, Target Vector Value:NiagaraPosition=$Particles.Position, Variable Type=::NewEnumerator2) {module=InterpolateOverTime001}
   /SetVariables(Particles.Debug:NiagaraFloat=$System.Timeline002.Time) {isSetParameters=true}
   &DynamicMaterialParameters(
    Index 0 Param 1=$Particles.Debug
    Module.Param0WriteEnabled=true
    Module.Param1WriteEnabled=true
    Module.Param2WriteEnabled=true
    Module.Param3WriteEnabled=true
   )
  ]
  rendererClasses [&NiagaraSpriteRendererProperties]
  renderers [{rendererClass=&NiagaraSpriteRendererProperties}]
 }
 Glow {
  emitterUpdateScript [
   &EmitterState(Life Cycle Mode=::NewEnumerator1, Loop Behavior=::NewEnumerator1, Loop Duration Mode=::NewEnumerator1, UseLoopDelay=false)
   &SpawnBurst_Instantaneous(Use Spawn Probability=false, UseLoopCountLimit=false)
  ]
  particleSpawnScript [
   &FN_SetSpawnGroup(Number of Spawn Groups=2)
   /Niagara/Modules/Spawn/Initialization/V2/InitializeParticle(
    Lifetime=1.1
    Mesh Scale=&Multiply_Vector(
     A=&VectorFromFloat(Value=0.9)
     B=&MaskVectorBySpawnGroup(Masked Vector=<-1 1 1>, Passthrough Vector=<1 1 1>)
    )
    Mesh Scale Mode=::NewEnumerator4
    Position Offset=<0 0 -5>
    Position Offset Coordinate Space=::Local
    UsePositionOffset=true
    Write Material Random=true
   )
   &CameraOffset() {enabled=false}
   &InitialMeshOrientation(Mesh Orientation Mode=::NewEnumerator7, Use Rotation Vector=true) {enabled=false, module=InitialMeshOrientation001}
   &InitialMeshOrientation(
    Mesh Orientation Mode=::NewEnumerator7
    Rotation=&Multiply_Vector(
     A=&Multiply_Vector(A=<-0.02 0 0.05>, B=&MakeVector(X=1, Y=1, Z=1))
     B=&MaskVectorBySpawnGroup(Masked Vector=<1 1 -1>, Passthrough Vector=<1 1 1>)
    )
    Use Rotation Vector=true
   )
  ]
  particleUpdateScript [
   &ParticleState(Kill Particles When Lifetime Has Elapsed=false, Let Infinitely Lived Particles Die When Emitter Deactivates=true)
   &SkeletalMeshLocation(Bone Sampling Mode=::NewEnumerator5, MaskSpawnGroup=false, Sampled Position Offset=<0 -10 4>, Skeletal Mesh=$System.Skeletal Mesh)
   &FN_Light_Attributes(
    Color=<rgba 0.6046 0.2291 1 1>
    Create Color Attribute=true
    Create Exponent Attribute=true
    Create Radius Attribute=true
    Exponent=4
    Radius=40
    Scale Color=&Multiply_Float(A=5, B=&Sine(Normalized Angle=$Emitter.Age, Period=0.2, Scale=0.05, Bias=1))
   )
  ]
  rendererClasses [&NiagaraLightRendererProperties]
  renderers [{rendererClass=&NiagaraLightRendererProperties, rendererIndex=0}]
 }
 Glow001 {
  emitterUpdateScript [
   &EmitterState(Life Cycle Mode=::NewEnumerator1, Loop Behavior=::NewEnumerator1, Loop Duration Mode=::NewEnumerator1, UseLoopDelay=false)
   &SpawnBurst_Instantaneous(Spawn Count=2, Use Spawn Probability=false, UseLoopCountLimit=false)
  ]
  particleSpawnScript [
   &FN_SetSpawnGroup(Number of Spawn Groups=2)
   &InitializeParticle(
    Lifetime=1.1
    Mesh Scale=&Multiply_Vector(
     A=&VectorFromFloat(Value=0.8)
     B=&MaskVectorBySpawnGroup(Masked Vector=<-1 -1 1>, Passthrough Vector=<1 1 1>)
    )
    Mesh Scale Mode=::NewEnumerator4
    Position Offset=<0 0 -5>
    Position Offset Coordinate Space=::Local
    UsePositionOffset=true
    Write Material Random=true
   )
   &InitialMeshOrientation(
    Mesh Orientation Mode=::NewEnumerator7
    Rotation=&Multiply_Vector(
     A=&Multiply_Vector(
      A=&Lerp_Vector(A=<-0.02 0 0.05>, B=<-0.0323 0 0.05>, Alpha=&FN_FrontEndFloatModifier(Gameplay=0, Front End=1))
      B=&MakeVector(X=1, Y=1, Z=1)
     )
     B=&MaskVectorBySpawnGroup(Masked Vector=<1 1 -1>, Passthrough Vector=<1 1 1>)
    )
    Use Rotation Vector=true
   )
  ]
  particleUpdateScript [
   &ParticleState(Kill Particles When Lifetime Has Elapsed=false, Let Infinitely Lived Particles Die When Emitter Deactivates=true)
   &SkeletalMeshLocation(Bone Sampling Mode=::NewEnumerator5, MaskSpawnGroup=false, Sampled Position Offset=<0 -2 -4>, Skeletal Mesh=$System.Skeletal Mesh)
   /SetVariables(Particles.MeshOrientation=&MultiplyQuaternion(
     Quaternion A=$StackContext.MeshOrientation
     Quaternion B=&MakeQuatFromAxisAngle(
      Axis=<0 0 1>
      AngleInDegrees=&Multiply_Float(
       A=&Multiply_Float(
        A=&Multiply_Float(A=20, B=&MaskFloatBySpawnGroup(Masked Float=-1, Passthrough Float=1))
        B=&Subtract_Float(A=1, B=$System.BackwardMovement)
       )
       B=$System.StateCheck
      )
     )
    )) {isSetParameters=true}
   /SetVariables(Particles.MeshOrientation=&MultiplyQuaternion(
     Quaternion A=$StackContext.MeshOrientation
     Quaternion B=&MakeQuatFromAxisAngle(
      Axis=<0 0 1>
      AngleInDegrees=&Multiply_Float(
       A=&Multiply_Float(
        A=&Subtract_Float(A=20, B=&Sine(Normalized Angle=$Emitter.Age, Period=1, Scale=5, Bias=1))
        B=&MaskFloatBySpawnGroup(Masked Float=-1, Passthrough Float=1)
       )
       B=&Lerp_Float(
        A=0
        B=$System.Timeline002.Time
        Alpha=&MakeCustomFloatFromBool(Bool=$User.IsSkydiving, True Float=1, False Float=0)
       )
      )
     )
    )) {isSetParameters=true}
   &InterpolateOverTime(
    Rate Of Change:NiagaraFloat=&FN_FrontEndFloatModifier(Gameplay=8, Front End=13)
    Target Vector Value:NiagaraPosition=$Particles.Position
    Variable Type=::NewEnumerator2
   ) {module=InterpolateOverTime001}
   /SetVariables(Particles.Interpolate Dynamic=&MakeFloat4FromVectorAndFloat("Vector (RGB)"=$StackContext.InterpolateOverTime001.InterpolateOverTime_V3_.Moving Average, "Float (Alpha)"=$User.Interpolate Intensity)) {isSetParameters=true}
   /SetVariables(Particles.WingClonePos:NiagaraPosition=&AddVectorToPosition(Position=$Particles.Position, Vector=&TransformVector(Vector=<0 0.02 0>, Destination Space=::Simulation))) {isSetParameters=true}
   &DynamicMaterialParameters(
    Index 0 Param 1=&Multiply_Float(
     A=&FloatFromCurve(
      FloatCurve=DataInterface() {propertyValues {Curve {keys [<at 0 1>,<at 0.8 0.4>]}}}
      CurveIndex=$System.Timeline.Time
      Scale Curve=&FN_FrontEndFloatModifier(Gameplay=$User.Interpolate Clamp, Front End=10)
     )
     B=&Multiply_Float(
      A=&FloatFromCurve(FloatCurve=DataInterface() {propertyValues {Curve {keys [<at 0 1>,<at 1 0>]}}}, CurveIndex=$System.Timeline001.Time)
      B=&Multiply_Float(
       A=&MakeCustomFloatFromBool(Bool=$User.IsSkydiving, True Float=0, False Float=1)
       B=&MakeCustomFloatFromBool(Bool=$User.bisInVehicle, True Float=0.3, False Float=1)
      )
     )
    )
    Index 0 Param 2=&MaskFloatBySpawnGroup(Masked Float=0, Passthrough Float=0.2)
    Index 0 Param 3=&MaskFloatBySpawnGroup(Masked Float=0, Passthrough Float=0.3)
    Module.Param0WriteEnabled=true
    Module.Param1WriteEnabled=true
    Module.Param2WriteEnabled=true
    Module.Param3WriteEnabled=true
   )
   &ScaleColor(
    Scale RGB=&VectorFromFloat(Value=&Multiply_Float(
      A=&FN_FrontEndFloatModifier(Gameplay=1, Front End=0.75)
      B=&FNTimeOfDayScalar(Night Scale=0.8, Morning Scale=0.8, Evening Scale=0.8)
     ))
    ScaleA=true
    ScaleRGB=true
   )
  ]
  rendererClasses [&NiagaraMeshRendererProperties]
  renderers [
   {rendererClass=&NiagaraMeshRendererProperties, rendererIndex=0}
   {rendererClass=&NiagaraMeshRendererProperties, rendererIndex=1}
  ]
 }
}
summary {
 systemName=NS_IchorIncisor_Idle
 emitters [
  {emitterName=Glow001, rendererClasses [&NiagaraMeshRendererProperties]}
  {bEnabled=false, emitterName=FN_Debug, rendererClasses [&NiagaraSpriteRendererProperties]}
  {emitterName=Glow, rendererClasses [&NiagaraLightRendererProperties]}
 ]
 userVariables [
  User.bIsFrontEnd {type=NiagaraBool}
  User.bIsFrontEndPreview {type=NiagaraBool}
  User.bisInVehicle {type=NiagaraBool}
  User.Interpolate Clamp {type=NiagaraFloat, defaultValue=10, description=""}
  User.Interpolate Intensity {type=NiagaraFloat, defaultValue=1.75, description=""}
  User.IsSkydiving {type=NiagaraBool}
 ]
}
